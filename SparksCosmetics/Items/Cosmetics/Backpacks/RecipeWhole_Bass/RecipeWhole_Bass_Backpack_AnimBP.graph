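# Anim Blueprint for the RecipeWhole Bass back bling, adjusting how the instrument's skeleton is posed.
# A chain of additive Transform (Modify) Bone nodes nudges the offset bone's position and rotation, then rotates base_ear_r.
# One offset step reads bIsSprinting as its alpha with a one-second blend, though the asset declares no such variable.
# The Event Graph holds only unconnected update and pawn-owner nodes.
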AnimBlueprint /SparksCosmetics/Items/Cosmetics/Backpacks/RecipeWhole_Bass/RecipeWhole_Bass_Backpack_AnimBP
target_skeleton=/SparksCosmetics/Instrument/Bass/RecipeWhole/Meshes/RecipeWhole_Bass_Skeleton
graphs [
 AnimGraph {nodes [
   n1 = "Transform (Modify) Bone - Bone: offset"(ComponentPose=Local To Component(LocalPose=Input Pose()), Alpha=1, Translation="0,0.5,0", Rotation="0,10,5", Scale="1,1,1") {rotationMode=BMM_Additive, translationMode=BMM_Additive, boneToModify {boneName=offset}, rotation {roll=5, yaw=10}, translation {y=0.5}}
   n2 = "Transform (Modify) Bone - Bone: offset"!(ComponentPose=@n1, bAlphaBoolEnabled:bool=Get bIsSprinting(), Translation="0,3,0", Rotation="0,0,6.231", Scale="1,1,1") -> (Pose) {
    alphaInputType=Bool
    rotationMode=BMM_Additive
    translationMode=BMM_Additive
    alphaBoolBlend {blendInTime=1, blendOutTime=1}
    boneToModify {boneName=offset}
    rotation {roll=6.231}
    translation {y=3}
   }
   n3 = "Transform (Modify) Bone - Bone: offset"(ComponentPose=@n2, bAlphaBoolEnabled=True, Translation="5,-0.09,4.158", Rotation="12.700,-7.418,2.025", Scale="1,1,1") {
    alphaInputType=Bool
    rotationMode=BMM_Additive
    rotationSpace=BCS_BoneSpace
    scaleMode=BMM_Additive
    scaleSpace=BCS_BoneSpace
    translationMode=BMM_Additive
    translationSpace=BCS_BoneSpace
    alphaBoolBlend {blendOption=Linear}
    boneToModify {boneName=offset}
    rotation=<rot 12.7 -7.418 2.025>
    translation=<5 -0.09 4.158>
   }
   n4 = "Transform (Modify) Bone - Bone: base_ear_r"!(ComponentPose=@n3, bAlphaBoolEnabled:bool=True, Translation="0,0,0", Rotation="0,15,0", Scale="1,1,1") -> (Pose) {
    alphaInputType=Bool
    rotationMode=BMM_Additive
    rotationSpace=BCS_BoneSpace
    scaleMode=BMM_Additive
    scaleSpace=BCS_BoneSpace
    translationMode=BMM_Additive
    translationSpace=BCS_BoneSpace
    alphaBoolBlend {blendOption=Linear}
    boneToModify {boneName=base_ear_r}
    rotation {yaw=15}
   }
   Output Pose(Result=Component To Local(ComponentPose=@n4))
  ]}
 EventGraph {nodes [
   TryGetPawnOwner()
   Event BlueprintUpdateAnimation()
  ]}
]
variables []
